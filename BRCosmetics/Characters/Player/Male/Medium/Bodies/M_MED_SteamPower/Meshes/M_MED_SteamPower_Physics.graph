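# Physics asset for the SteamPower medium male player body cosmetic, set up for four dangling bottles around the hips.
# Each dyn_bottle capsule hangs from a small base_bottle anchor with locked twist and limited swing, and each anchor links to the pelvis.
# Pelvis, lower spine and double-capsule thighs provide body shapes around the bottles, with free angular constraints to the pelvis.
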PhysicsAsset /BRCosmetics/Characters/Player/Male/Medium/Bodies/M_MED_SteamPower/Meshes/M_MED_SteamPower_Physics
bodies {
 base_bottle_01_l [{shapeType=Capsule, length=0, radius=0.505, shapeName=base_bottle_01_l_capsule, rotation=<rot 0 -90 0>}]
 base_bottle_01_r [{shapeType=Capsule, length=0.005, radius=0.505, shapeName=base_bottle_01_r_capsule, rotation=<rot 0 -90 0>}]
 base_bottle_02_l [{shapeType=Capsule, length=0.005, radius=0.505, shapeName=base_bottle_02_l_capsule, rotation=<rot 0 -90 0>}]
 base_bottle_02_r [{shapeType=Capsule, length=0.005, radius=0.505, shapeName=base_bottle_02_r_capsule, rotation=<rot 0 -90 0>}]
 dyn_bottle_01_l [{shapeType=Capsule, length=0.5, radius=2.0562, shapeName=dyn_bottle_01_r_capsule, center=<7.8518 0 0>, rotation=<rot 0 90 -90>}]
 dyn_bottle_01_r [{shapeType=Capsule, length=1, radius=2.0562, shapeName=dyn_bottle_01_r_capsule, center=<7.9511 0 0>, rotation=<rot 0 90 -90>}]
 dyn_bottle_02_l [{shapeType=Capsule, length=4.9246, radius=1.0007, shapeName=dyn_bottle_02_l_capsule, center=<5.2097 0 0>, rotation=<rot 0 90 -90>}]
 dyn_bottle_02_r [{shapeType=Capsule, length=3.9075, radius=1.2662, shapeName=dyn_bottle_02_l_capsule, center=<4.8492 0 0>, rotation=<rot 0 90 -90>}]
 pelvis [{shapeType=Capsule, length=4.4521, radius=14.6946, shapeName=pelvis_capsule, center=<-11.0467 -1.7836 0>, rotation=<rot 0 0 180>}]
 spine_01 [{shapeType=Capsule, length=3, radius=15.25, shapeName=spine_01_capsule, center=<0.6034 -3.5368 -0.4571>, rotation=<rot 90 180 178.4554>}]
 thigh_l [
  {shapeType=Capsule, length=9.5, radius=9, shapeName=thigh_l_capsule, center=<-21.56 -0.8176 0.3293>, rotation=<rot 90 0 15.0001>}
  {shapeType=Capsule, length=9.5, radius=9, shapeName=thigh_l_capsule, center=<-14.7951 -2.7097 0.3293>, rotation=<rot 90 0 0>}
 ]
 thigh_r [
  {shapeType=Capsule, length=10, radius=8.5, shapeName=thigh_r_capsule, center=<21.2283 -0.6233 0>, rotation=<rot 90 172.875 -172.1248>}
  {shapeType=Capsule, length=10, radius=8.5, shapeName=thigh_r_capsule, center=<16.3976 0.659 0>, rotation=<rot 90 0 0>}
 ]
}
constraints [
 {
  bone1Name=base_bottle_01_l
  bone2Name=pelvis
  swing1LimitDegrees=45
  swing1Motion=Free
  swing2LimitDegrees=45
  swing2Motion=Free
  twistLimitDegrees=45
  twistMotion=Free
 }
 {
  bone1Name=dyn_bottle_01_l
  bone2Name=base_bottle_01_l
  swing1LimitDegrees=60
  swing2LimitDegrees=50
  twistLimitDegrees=45
  twistMotion=Locked
 }
 {
  bone1Name=base_bottle_01_r
  bone2Name=pelvis
  swing1LimitDegrees=45
  swing1Motion=Free
  swing2LimitDegrees=45
  swing2Motion=Free
  twistLimitDegrees=45
  twistMotion=Free
 }
 {
  bone1Name=dyn_bottle_01_r
  bone2Name=base_bottle_01_r
  swing1LimitDegrees=60
  swing2LimitDegrees=45
  twistLimitDegrees=45
  twistMotion=Locked
 }
 {
  bone1Name=base_bottle_02_l
  bone2Name=pelvis
  swing1LimitDegrees=45
  swing1Motion=Free
  swing2LimitDegrees=45
  swing2Motion=Free
  twistLimitDegrees=45
  twistMotion=Free
 }
 {
  bone1Name=dyn_bottle_02_l
  bone2Name=base_bottle_02_l
  swing1LimitDegrees=60
  swing2LimitDegrees=60
  twistLimitDegrees=45
  twistMotion=Locked
 }
 {
  bone1Name=base_bottle_02_r
  bone2Name=pelvis
  swing1LimitDegrees=45
  swing1Motion=Free
  swing2LimitDegrees=45
  swing2Motion=Free
  twistLimitDegrees=45
  twistMotion=Free
 }
 {
  bone1Name=dyn_bottle_02_r
  bone2Name=base_bottle_02_r
  swing1LimitDegrees=60
  swing2LimitDegrees=60
  twistLimitDegrees=45
  twistMotion=Locked
 }
 {
  bone1Name=spine_01
  bone2Name=pelvis
  swing1LimitDegrees=45
  swing1Motion=Free
  swing2LimitDegrees=45
  swing2Motion=Free
  twistLimitDegrees=45
  twistMotion=Free
 }
 {
  bone1Name=thigh_l
  bone2Name=pelvis
  swing1LimitDegrees=45
  swing1Motion=Free
  swing2LimitDegrees=45
  swing2Motion=Free
  twistLimitDegrees=45
  twistMotion=Free
 }
 {
  bone1Name=thigh_r
  bone2Name=pelvis
  swing1LimitDegrees=45
  swing1Motion=Free
  swing2LimitDegrees=45
  swing2Motion=Free
  twistLimitDegrees=45
  twistMotion=Free
 }
]
