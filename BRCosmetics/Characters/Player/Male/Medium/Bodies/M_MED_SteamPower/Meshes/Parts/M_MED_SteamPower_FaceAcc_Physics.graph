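# Physics asset for the male medium SteamPower FaceAcc mesh, setting up capsule bodies and constraints for hair motion around the head.
# Small base anchor bodies join a two-capsule head body with free angular motion.
# From these anchors hang eight single-segment back strands and three three-segment front chains (A, B, C).
# Dynamic segments use narrow swing limits with zero twist, generally widening toward the front-chain tips.
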PhysicsAsset /BRCosmetics/Characters/Player/Male/Medium/Bodies/M_MED_SteamPower/Meshes/Parts/M_MED_SteamPower_FaceAcc_Physics
bodies {
 base_hair_back_strand_01 [{shapeType=Capsule, length=0.295, radius=0.505, shapeName=base_hair_back_strand_01_capsule, rotation=<rot 0 -90 0>}]
 base_hair_back_strand_02 [{shapeType=Capsule, length=0.295, radius=0.505, shapeName=base_hair_back_strand_02_capsule, rotation=<rot 0 -90 0>}]
 base_hair_back_strand_03 [{shapeType=Capsule, length=0.295, radius=0.505, shapeName=base_hair_back_strand_03_capsule, rotation=<rot 0 -90 0>}]
 base_hair_back_strand_04 [{shapeType=Capsule, length=0.295, radius=0.505, shapeName=base_hair_back_strand_04_capsule, rotation=<rot 0 -90 0>}]
 base_hair_back_strand_05 [{shapeType=Capsule, length=0.295, radius=0.505, shapeName=base_hair_back_strand_05_capsule, rotation=<rot 0 -90 0>}]
 base_hair_back_strand_06 [{shapeType=Capsule, length=0.295, radius=0.505, shapeName=base_hair_back_strand_06_capsule, rotation=<rot 0 -90 0>}]
 base_hair_back_strand_07 [{shapeType=Capsule, length=0.295, radius=0.505, shapeName=base_hair_back_strand_07_capsule, rotation=<rot 0 -90 0>}]
 base_hair_back_strand_08 [{shapeType=Capsule, length=0.505, radius=0.505, shapeName=base_hair_back_strand_08_capsule, rotation=<rot 0 -90 0>}]
 base_hair_front_A [{shapeType=Capsule, length=0.505, radius=0.505, shapeName=base_hair_front_A_capsule, rotation=<rot 0 -90 0>}]
 base_hair_front_B [{shapeType=Capsule, length=0.505, radius=0.505, shapeName=base_hair_front_B_capsule, rotation=<rot 0 -90 0>}]
 base_hair_front_C [{shapeType=Capsule, length=0.505, radius=0.505, shapeName=base_hair_front_C_capsule, rotation=<rot 0 -90 0>}]
 dyn_hair_back_strand_01 [{shapeType=Capsule, length=2.5287, radius=1, shapeName=dyn_hair_back_strand_01_capsule, center=<2.3322 0 0>, rotation=<rot 90 0 0>}]
 dyn_hair_back_strand_02 [{shapeType=Capsule, length=1.8891, radius=1, shapeName=dyn_hair_back_strand_02_capsule, center=<1.609 0 0>, rotation=<rot 90 0 0>}]
 dyn_hair_back_strand_03 [{shapeType=Capsule, length=1.7422, radius=1.2634, shapeName=dyn_hair_back_strand_03_capsule, center=<1.8216 0 0>, rotation=<rot 90 0 0>}]
 dyn_hair_back_strand_04 [{shapeType=Capsule, length=1.3407, radius=1, shapeName=dyn_hair_back_strand_04_capsule, center=<1.1936 0 0>, rotation=<rot 90 0 0>}]
 dyn_hair_back_strand_05 [{shapeType=Capsule, length=0.9853, radius=1, shapeName=dyn_hair_back_strand_05_capsule, center=<0.9037 0 0>, rotation=<rot 90 0 0>}]
 dyn_hair_back_strand_06 [{shapeType=Capsule, length=2.7682, radius=1, shapeName=dyn_hair_back_strand_06_capsule, center=<2.628 0 0>, rotation=<rot 90 0 0>}]
 dyn_hair_back_strand_07 [{shapeType=Capsule, length=0.9998, radius=1.6156, shapeName=dyn_hair_back_strand_07_capsule, center=<1.6719 0 0>, rotation=<rot 90 0 0>}]
 dyn_hair_back_strand_08 [{shapeType=Capsule, length=4.8643, radius=0.9013, shapeName=dyn_hair_back_strand_08_capsule, center=<2.8692 0 0>, rotation=<rot 0 90 -90>}]
 dyn_hair_front_A_01 [{shapeType=Capsule, length=4.4086, radius=1, shapeName=dyn_hair_front_A_01_capsule, center=<2.5619 0 0.6145>, rotation=<rot 90 0 0>}]
 dyn_hair_front_A_02 [{shapeType=Capsule, length=1.699, radius=1.5326, shapeName=dyn_hair_front_A_02_capsule, center=<2.4967 0 0.17>, rotation=<rot 85.0001 0 0>}]
 dyn_hair_front_A_03 [{
   shapeType=Capsule
   length=3.9567
   radius=1.3451
   shapeName=dyn_hair_front_A_03_capsule
   center=<3.1037 0 -0.3878>
   rotation=<rot 75.0001 0 0>
  }]
 dyn_hair_front_B_01 [{shapeType=Capsule, length=2.0659, radius=1.0355, shapeName=dyn_hair_front_B_01_capsule, center=<2.1969 0 0.7719>, rotation=<rot 90 0 0>}]
 dyn_hair_front_B_02 [{shapeType=Capsule, length=2.58, radius=1.3904, shapeName=dyn_hair_front_B_02_capsule, center=<2.4481 0 1.0697>, rotation=<rot 90 0 0>}]
 dyn_hair_front_B_03 [{shapeType=Capsule, length=4.811, radius=1.3454, shapeName=dyn_hair_front_B_03_capsule, center=<2.9328 -1.15 0>, rotation=<rot 90 0 0>}]
 dyn_hair_front_C_01 [{shapeType=Capsule, length=2.4763, radius=1.1392, shapeName=dyn_hair_front_C_01_capsule, center=<2.5057 0 0>, rotation=<rot 0 -90 90>}]
 dyn_hair_front_C_02 [{shapeType=Capsule, length=3.477, radius=1.2073, shapeName=dyn_hair_front_A_03_capsule, center=<3.1037 -0.5 0>, rotation=<rot 0 90 90>}]
 dyn_hair_front_C_03 [{shapeType=Capsule, length=3.9484, radius=1.3533, shapeName=dyn_hair_front_A_03_capsule, center=<3.1037 -0.3 -0.35>, rotation=<rot 0 90 90>}]
 head [
  {shapeType=Capsule, length=6.3947, radius=7.8295, center=<5.1801 2.564 0.0193>, rotation=<rot 90 179.9999 -179.9998>}
  {shapeType=Capsule, length=0, radius=7.6156, center=<9.826 2.9094 0.1533>, rotation=<rot 0 -0.0004 0>}
 ]
}
constraints [
 {bone1Name=dyn_hair_back_strand_01, bone2Name=base_hair_back_strand_01, swing1LimitDegrees=30, swing2LimitDegrees=5, twistLimitDegrees=0}
 {bone1Name=dyn_hair_back_strand_02, bone2Name=base_hair_back_strand_02, swing1LimitDegrees=20, swing2LimitDegrees=5, twistLimitDegrees=0}
 {bone1Name=dyn_hair_back_strand_03, bone2Name=base_hair_back_strand_03, swing1LimitDegrees=20, swing2LimitDegrees=20, twistLimitDegrees=0}
 {bone1Name=dyn_hair_back_strand_04, bone2Name=base_hair_back_strand_04, swing1LimitDegrees=10, swing2LimitDegrees=10, twistLimitDegrees=0}
 {bone1Name=dyn_hair_back_strand_05, bone2Name=base_hair_back_strand_05, swing1LimitDegrees=20, swing2LimitDegrees=5, twistLimitDegrees=0}
 {bone1Name=dyn_hair_back_strand_06, bone2Name=base_hair_back_strand_06, swing1LimitDegrees=15, swing2LimitDegrees=5, twistLimitDegrees=0}
 {bone1Name=dyn_hair_back_strand_07, bone2Name=base_hair_back_strand_07, swing1LimitDegrees=20, swing2LimitDegrees=20, twistLimitDegrees=0}
 {bone1Name=dyn_hair_front_A_01, bone2Name=base_hair_front_A, swing1LimitDegrees=3, swing2LimitDegrees=2, twistLimitDegrees=0}
 {bone1Name=dyn_hair_front_A_02, bone2Name=dyn_hair_front_A_01, swing1LimitDegrees=7, swing2LimitDegrees=15, twistLimitDegrees=0}
 {bone1Name=dyn_hair_front_A_03, bone2Name=dyn_hair_front_A_02, swing1LimitDegrees=12, swing2LimitDegrees=20, twistLimitDegrees=0}
 {bone1Name=dyn_hair_front_B_01, bone2Name=base_hair_front_B, swing1LimitDegrees=5, swing2LimitDegrees=2, twistLimitDegrees=0}
 {bone1Name=dyn_hair_front_B_02, bone2Name=dyn_hair_front_B_01, swing1LimitDegrees=5, swing2LimitDegrees=2, twistLimitDegrees=0}
 {bone1Name=dyn_hair_front_B_03, bone2Name=dyn_hair_front_B_02, swing1LimitDegrees=10, swing2LimitDegrees=5, twistLimitDegrees=0}
 {
  bone1Name=base_hair_back_strand_07
  bone2Name=head
  swing1LimitDegrees=45
  swing1Motion=Free
  swing2LimitDegrees=45
  swing2Motion=Free
  twistLimitDegrees=45
  twistMotion=Free
 }
 {
  bone1Name=base_hair_back_strand_06
  bone2Name=head
  swing1LimitDegrees=45
  swing1Motion=Free
  swing2LimitDegrees=45
  swing2Motion=Free
  twistLimitDegrees=45
  twistMotion=Free
 }
 {
  bone1Name=base_hair_back_strand_05
  bone2Name=head
  swing1LimitDegrees=45
  swing1Motion=Free
  swing2LimitDegrees=45
  swing2Motion=Free
  twistLimitDegrees=45
  twistMotion=Free
 }
 {
  bone1Name=base_hair_back_strand_04
  bone2Name=head
  swing1LimitDegrees=45
  swing1Motion=Free
  swing2LimitDegrees=45
  swing2Motion=Free
  twistLimitDegrees=45
  twistMotion=Free
 }
 {
  bone1Name=base_hair_back_strand_03
  bone2Name=head
  swing1LimitDegrees=45
  swing1Motion=Free
  swing2LimitDegrees=45
  swing2Motion=Free
  twistLimitDegrees=45
  twistMotion=Free
 }
 {
  bone1Name=base_hair_back_strand_02
  bone2Name=head
  swing1LimitDegrees=45
  swing1Motion=Free
  swing2LimitDegrees=45
  swing2Motion=Free
  twistLimitDegrees=45
  twistMotion=Free
 }
 {
  bone1Name=base_hair_back_strand_01
  bone2Name=head
  swing1LimitDegrees=45
  swing1Motion=Free
  swing2LimitDegrees=45
  swing2Motion=Free
  twistLimitDegrees=45
  twistMotion=Free
 }
 {
  bone1Name=base_hair_front_A
  bone2Name=head
  swing1LimitDegrees=45
  swing1Motion=Free
  swing2LimitDegrees=45
  swing2Motion=Free
  twistLimitDegrees=45
  twistMotion=Free
 }
 {
  bone1Name=base_hair_front_B
  bone2Name=head
  swing1LimitDegrees=45
  swing1Motion=Free
  swing2LimitDegrees=45
  swing2Motion=Free
  twistLimitDegrees=45
  twistMotion=Free
 }
 {bone1Name=dyn_hair_front_C_01, bone2Name=base_hair_front_C, swing1LimitDegrees=2, swing2LimitDegrees=2, twistLimitDegrees=0}
 {bone1Name=dyn_hair_front_C_02, bone2Name=dyn_hair_front_C_01, swing1LimitDegrees=8, swing2LimitDegrees=5, twistLimitDegrees=0}
 {bone1Name=dyn_hair_front_C_03, bone2Name=dyn_hair_front_C_02, swing1LimitDegrees=12, swing2LimitDegrees=5, twistLimitDegrees=0}
 {
  bone1Name=base_hair_front_C
  bone2Name=head
  swing1LimitDegrees=45
  swing1Motion=Free
  swing2LimitDegrees=45
  swing2Motion=Free
  twistLimitDegrees=45
  twistMotion=Free
 }
 {
  bone1Name=base_hair_back_strand_08
  bone2Name=head
  swing1LimitDegrees=45
  swing1Motion=Free
  swing2LimitDegrees=45
  swing2Motion=Free
  twistLimitDegrees=45
  twistMotion=Free
 }
 {bone1Name=dyn_hair_back_strand_08, bone2Name=base_hair_back_strand_08, swing1LimitDegrees=7, swing2LimitDegrees=5, twistLimitDegrees=0}
]
